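; Tank main cannon projectile that reports firing to the tank, spins in flight and explodes on stop or fuse timeout.
; Gravity applies after a delay, with both values read from scalable floats at construction.
; Valid touches either double velocity on NoStick components or flag stop-on-hit and record the surface type.
; Explosion cosmetics pick water or standard FX and sound, trigger camera shake and optionally spawn a physics-field radial force.
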(event EventBeginPlay
  ;[1 1 1] Begin Play
  (Utilities|Time|SetTimerbyEvent (delegate Custom|EnableGravity) (Variables|Default|GetProjectileGravityDelay))
  (Utilities|Time|SetTimerbyFunctionName self "FuseTimerMax" 6.0)
  (bind _astank_vehicle (Utilities|Casting|CastToTank_Vehicle (Vehicle|GetVehicleActor (Utilities|Casting|CastToFortPlayerPawn (Game|GetInstigator)) :!class /Script/FortniteGame.FortPlayerPawn)))
  (Class|TankVehicle|CannonFired _astank_vehicle self)
  (Networking|CanExecuteCosmeticEvents))

(event Custom|EnableGravity
  ;[1 1 1] Begin Play
  (Variables|/Script/Engine.ProjectileMovementComponent|SetProjectileGravityScale (Variables|Default|GetProjectileGravity) (Variables|Movement|GetProjectileMovementComponent)))

(event EventTick
  (params (DeltaSeconds :type Float%20%28single-precision%29))
  ;[1 1 1] On Tick, spin the rocket and particle trail
  (Variables|Default|SetTick%20Delta DeltaSeconds)
  (Networking|CanExecuteCosmeticEvents
    (:True
      (Variables|Effects|SetRocketSpin (Math|Rotator|MakeRotator 0.0 (* (Variables|Default|GetTick%20Delta) (Variables|Effects|GetRocketSpinRate))))
      (Transformation|AddLocalRotation (Variables|Default|GetMissile) (Variables|Effects|GetRocketSpin) false false))))

(event EventOnTouched
  (params (OtherActor :type Actor%20Object%20Reference) (OtherComp :type Primitive%20Component%20Object%20Reference) (HitResult :type Hit%20Result%20Structure%20%28by%20ref%29) (bIsOverlap :type Boolean))
  (bind (_bblockinghit _binitialoverlap _time _distance _location _impactpoint _normal _impactnormal _physmat) (Collision|BreakHitResult HitResult))
  (bind _surfacetype (Variables|/Script/PhysicsCore.PhysicalMaterial|GetSurfaceType _physmat))
  (bind (_bblockinghit_2 _binitialoverlap_2 _time_2 _distance_2 _location_2 _impactpoint_2 _normal_2 _impactnormal_2 _physmat_2 _hitactor _hitcomponent) (Collision|BreakHitResult HitResult))
  ;[1 1 1] Bounce off "NoBounce"
  (Utilities|IsValid _physmat
    (:Is%20Valid
      (if (or (Utilities|Enum|Equal(Enum) _surfacetype "SurfaceType8") (not bIsOverlap))
        (if (Components|ComponentHasTag _hitcomponent "NoStick")
          (Variables|/Script/Engine.MovementComponent|SetVelocity (* (Variables|/Script/Engine.MovementComponent|GetVelocity (Variables|Movement|GetProjectileMovementComponent)) 2.0) (Variables|Movement|GetProjectileMovementComponent))
          (else
            (Variables|Projectile|SetbStopSimulatingOnHit true)
            (Variables|Default|SetSurfaceType _surfacetype)))))))

(event EventOnStop
  (params (Hit :type Hit%20Result%20Structure%20%28by%20ref%29))
  (bind (_bblockinghit _binitialoverlap _time _distance _location _impactpoint _normal _impactnormal _physmat) (Collision|BreakHitResult Hit))
  ;[1 1 1] Impact and Explosion Stuff
  ;;[1 1 1] Event --> On Stop
  (Variables|Effects|SetDecalLocation _impactpoint)
  (Variables|Effects|SetStopLocation (+ _impactpoint (* _impactnormal (Variables|Effects|GetExplosion%20Impact%20Offset))))
  (Variables|Effects|SetStopRotZ (Math|Rotator|MakeRotfromZ _impactnormal))
  (Utilities|IsValid _physmat
    (:Is%20Valid
      (Variables|Default|SetSurfaceType (Variables|/Script/PhysicsCore.PhysicalMaterial|GetSurfaceType _physmat)))
    (:Is%20Not%20Valid
      _))
  (if (not (Delivery|IsBeingKilled))
    (Delivery|DoExplosion)
    (Fortnite|ShowSoundIndicatorLocal :Location (Transformation|GetWorldLocation (Variables|Default|GetMissile)) :MaxAudibleDistance 5e+03 :IndicatorType "Gunshot" :Instigator self :Tint "(R=0.000000,G=0.000000,B=0.000000,A=0.000000)" :AttenuationMaxRangeMultiplier 1.0 :AttenuationIntensityMultiplier 1.0)))

(event Custom|FuseTimerMax
  (bind _missile (Variables|Default|GetMissile))
  (bind _returnvalue (Transformation|GetWorldLocation _missile))
  ;[1 1 1] Impact and Explosion Stuff
  ;;[1 1 1] Event --> On Stop
  (Variables|Effects|SetTimerMaxReached? true)
  (Variables|Effects|SetStopLocation _returnvalue)
  (if (not (Delivery|IsBeingKilled))
    (Delivery|DoExplosion)
    (Fortnite|ShowSoundIndicatorLocal :Location _returnvalue :MaxAudibleDistance 5e+03 :IndicatorType "Gunshot" :Instigator self :Tint "(R=0.000000,G=0.000000,B=0.000000,A=0.000000)" :AttenuationMaxRangeMultiplier 1.0 :AttenuationIntensityMultiplier 1.0)))

(event EventOnExploded
  (params (HitActors :type Array%20of%20Actor%20Object%20References) (HitResults :type Array%20of%20Hit%20Result%20Structures))
  (bind _stoplocation (Variables|Effects|GetStopLocation))
  (bind _stoprotz (Variables|Effects|GetStopRotZ))
  ;[1 1 1] Impact and Explosion Stuff
  (Networking|CanExecuteCosmeticEvents
    (:True
      ;;[1 1 1] SingleLineCheckForWorldObjectsSurfaceType
      (if (or (Utilities|Enum|Equal(Enum) (Variables|Default|GetSurfaceType) "SurfaceType8") (Water|IsinWater (Variables|Water|GetWaterInteractionComponent)))
        (bind _returnvalue (Effects|Components|ParticleSystem|SpawnFXAtLocation (Variables|Default|GetFX_Water_Explosion_Asset) _stoplocation _stoprotz "1.000000,1.000000,1.000000" true "AutoRelease"))
        (Rendering|SetRenderCustomDepth _returnvalue true)
        (Audio|PlaySoundatLocation (Variables|Sounds|GetRocket_Projectile_Explosion_Water_Sound) (Variables|Effects|GetStopLocation))
        (Audio|PlaySoundatLocation "/Game/Athena/Sounds/Weapons/Explosions/Expl_Debris_Water_Cue.Expl_Debris_Water_Cue" (Variables|Effects|GetStopLocation))
        (else
          (bind _returnvalue_2 (Effects|Components|ParticleSystem|SpawnFXAtLocation (Variables|Default|GetFX_Explosion_Asset) _stoplocation _stoprotz "1.000000,1.000000,1.000000" true "AutoRelease"))
          (Rendering|SetRenderCustomDepth _returnvalue_2 true)
          (Audio|PlaySoundatLocation (Variables|Sounds|GetRocket_Projectile_Explosion_Sound) (Variables|Effects|GetStopLocation))))
      (Rendering|SetVisibility (Variables|Default|GetMissile) false false)
      ;[1 1 1] Shake that camera
      (GameplayCue|ExecuteGameplayCue%28Burst%29Local self (Variables|Effects|GetFeedbackCue) (Ability|GameplayCue|MakeGameplayCueParameters 0.0)) ; Cam shake + controller vibration
      (Audio|Components|Audio|FadeOut (Variables|Default|GetRocket_AudioComponent) 1 0.0 "Linear" :!class /Script/Engine.AudioComponent)
      ;[1 1 1] FN Radial Force
      (Utilities|AllowJanus
        (:Allow
          (bind _returnvalue_3 (Development|GetConsoleVariableIntValue "r.PhysicsField.EnableField"))
          (if (== _returnvalue_3 1)
            (Game|SpawnActorfromClass "/Game/Effects/Niagara/Blueprints/PhsyicsFields/FN_RadialForce.FN_RadialForce_C" (Transformation|GetWorldTransform (Variables|Collision|GetCapsuleComponent)) "Undefined" "MultiplyWithRoot" 0 0.2 0.75 1.5e+03 2.0 0.0 0)))))
    (:False
      (CallFunction|HideonExplode))))

(event EventDestroyed
  ;[1 1 1] Impact and Explosion Stuff
  (Networking|CanExecuteCosmeticEvents
    (:True
      (Components|Activation|Deactivate (Variables|Default|GetNS_Tank_Cannon_Prj)))))

(event Custom|HideOnExplode
  ;[1 1 1] Impact and Explosion Stuff
  (Development|SetHiddeninGame (Variables|Default|GetMissile) true false)
  (Components|Activation|Deactivate (Variables|Default|GetNS_Tank_Cannon_Prj)))

(fn ConstructionScript
  (Effects|Components|ParticleSystem|SetTemplate (Variables|Default|GetRibbon_Trail_PSC) (Variables|Effects|GetRibbon_Trail_PS))
  (Variables|Default|SetFX_Explosion_Asset (Variables|Default|GetFX_Explosion_Asset))
  (Effects|Components|ParticleSystem|SetTemplate (Variables|Default|GetExplosion_Flesh_Damage_PSC) (Variables|Effects|GetExplosion_Flesh_Damage_PS))
  (Variables|Default|SetFX_Water_Explosion_Asset (Variables|Default|GetFX_Water_Explosion_Asset))
  (Effects|Components|ParticleSystem|SetTemplate (Variables|Default|GetWater_Explosion_PSC) (Variables|Effects|GetWater_Explosion_PS))
  (Variables|Default|SetProjectileGravityDelay (ScalableFloat|GetValueatLevel (LocalVariables|GetRow_ProjectileGravityDelay)))
  (Variables|Default|SetProjectileGravity (ScalableFloat|GetValueatLevel (LocalVariables|GetRow_ProjectileGravity))))
