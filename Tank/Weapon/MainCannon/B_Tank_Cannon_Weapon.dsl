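; Weapon Blueprint for the tank's main cannon, handling firing effects and scoped aiming.
; Firing triggers camera shake, muzzle FX, turret-aligned dust, an optional physics force field and a reload sound timed from fire rate.
; Scoping uses short timers to enable thermal post-process and scope mesh, hiding tank and passengers locally, with cached pawns for restoration after destruction.
; Muzzle effects choose Niagara or Cascade, async-loading Niagara assets.
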(event EventEndPlay
  (params (EndPlayReason :type EEndPlayReason%20Enum :default "Destroyed"))
  (|CallParentFunction EndPlayReason))

(event Custom|LoadNiagaraMuzzleSoftObject
  (bind _outniagarasystem (CallFunction|GetCorrectMuzzleNiagaraSystem))
  ;[1 1 1] Load Niagara Muzzle Soft Object
  (if (and (not (Asset|IsAssetNull _outniagarasystem)) (not (Utilities|IsValid (Utilities|ResolveSoftReference _outniagarasystem) :!class /Script/Engine.KismetSystemLibrary)))
    (Utilities|AsyncLoadAsset _outniagarasystem)))

(event EventBeginPlay
  ;[0 0 0] EVENT -> Begin Play:   Send wind particle reference to the wind manager
  (if (Variables|Default|GetIs%20Wind%20Enabled)
    (Wind|AddWindParticleSystemComponent (Fortnite|GetWindManager) (Variables|Effects|GetMuzzleWindParticleSpawned))))

(event Custom|BindFireRateChange
  ;[1 1 1] If Auto weapon, Bind Rate of fire change event for fire sound pitch then call it one time to set it initially
  (switch Utilities|FlowControl|Switch|SwitchonEFortWeaponTriggerType (WeaponItemDefinition|GetWeaponDataTriggerType)
    (:Automatic
      (FortWeapon|BindEventtoOnWeaponRateOfFireChanged (delegate Custom|PitchUpOnRateOfFireChange))
      (CallFunction|PitchUponRateOfFireChange (FortWeapon|GetWeaponRateOfFireModifier)))))

(event Custom|ShellEjectionFixOn
  ;[1 1 1] Fix for Shell Ejection Issue on Quality Change
  (EventDispatchers|BindEventtoOnEffectsQualityChanged (Settings|GetFortGameUserSettings) (delegate Custom|Bind%20on%20Effects%20Quality)))

(event Custom|ShellEjectionOff
  ;[1 1 1] Fix for Shell Ejection Issue on Quality Change
  (EventDispatchers|UnbindEventfromOnEffectsQualityChanged (Settings|GetFortGameUserSettings) (delegate Custom|Bind%20on%20Effects%20Quality)))

(event Custom|PitchUpOnRateOfFireChange
  (params (NewRateOfFire :type Float%20%28single-precision%29))
  ;[1 1 1] If Auto weapon, Bind Rate of fire change event for fire sound pitch then call it one time to set it initially
  ;;[1 1 1] Pitch on weapon rate of fire change
  (FortWeapon|SetFireSoundPitchMultiplier (Math|Curves|GetFloatValue (Variables|Default|GetCurve_PitchOffset) NewRateOfFire)))

(event Custom|Bind%20on%20Effects%20Quality
  ;[1 1 1] Fix for Shell Ejection Issue on Quality Change
  (Utilities|FlowControl|Delay 0.4) ; REQUIRED to make work!
  (CallFunction|DeactivateAllShellsFX)
  (CallFunction|ActivateShellsFX true))

(event EventOnPlayWeaponFireFX
  (params (bPersistentFire :type Boolean) (bSecondaryFire :type Boolean))
  (bind (_returnvalue_roll _returnvalue_pitch _returnvalue_yaw) (Math|Rotator|BreakRotator (Transformation|GetSocketRotation (Class|FortVehicleInterface|GetSkeletalMeshComponent (Variables|Default|GetTankVehicle)) "Turret")))
  (bind _returnvalue (Game|GetInstigator))
  (bind _asfort_player_pawn (Utilities|Casting|CastToFortPlayerPawn _returnvalue))
  (bind _returnvalue_2 (Pawn|GetLocalViewingPlayerController _asfort_player_pawn))
  (bind _sound_reloadmaincannon (Variables|Default|GetSound_ReloadMainCannon))
  ;[0.4433 0 1] EVENT -> Play/Stop Weapon Fire FX
  (Utilities|IsValid _returnvalue_2
    (:Is%20Valid
      (CameraShakes|StartCameraShake :self (Variables|/Script/Engine.PlayerController|GetPlayerCameraManager _returnvalue_2) :ShakeClass (Variables|Default|GetCannonFireCameraShake) :Scale 1.0 :PlaySpace "CameraLocal"))
    (:Is%20Not%20Valid
      _))
  (Utilities|FlowControl|Sequence
    (:then_0
      (CallFunction|MuzzleFlashFX bPersistentFire)
      ;;[1 0.6177 0] Handle Muzzle Flash FX and "Reload" Audio
      (Niagara|SpawnSystemAttached (Variables|Default|GetMuzzleFlashFXNS) (Class|FortVehicleInterface|GetSkeletalMeshComponent (Variables|Default|GetTankVehicle)) "Muzzle" "0, 0, 0" "0, 0, 0" "SnapToTarget" true)
      (bind _asfort_player_pawn_2 (Utilities|Casting|CastToFortPlayerPawn (Game|GetInstigator)))
      (Utilities|FlowControl|Delay (- (/ 1.0 (FortWeapon|GetFiringRate (Variables|/Script/FortniteGame.FortPawn|GetCurrentWeapon _asfort_player_pawn_2))) (Variables|/Script/Engine.SoundBase|GetDuration _sound_reloadmaincannon))) ; Wait before we play the "reload sound"
      (Audio|PlaySoundatLocation _sound_reloadmaincannon (Transformation|GetActorLocation)))
    (:then_1
      (CallFunction|ActivateorDeactivateWindParticle true))
    (:then_2
      (if (Effects|ShouldPlayShellFX)
        (CallFunction|ActivateShellsFX (Variables|Effects%28Shells%29|GetUseShellsOnFire?))))
    (:then_3
      (FortWeapon|PlayForceFeedback false bPersistentFire))
    (:then_4
      (Niagara|SpawnSystematLocation (Variables|Default|GetCannonRumbleFXNS) (Transformation|GetActorLocation (Variables|Default|GetTankVehicle)) (Math|Rotator|MakeRotator _returnvalue_roll 0.0 _returnvalue_yaw)) ; Dust around tank on Cannon fire
      (Utilities|AllowJanus
        (:Allow
          (bind _returnvalue_3 (Development|GetConsoleVariableIntValue "r.PhysicsField.EnableField"))
          (if (== _returnvalue_3 1)
            (Game|SpawnActorfromClass "/Game/Effects/Niagara/Blueprints/PhsyicsFields/FN_RadialForce.FN_RadialForce_C" (Transformation|GetActorTransform (Variables|Default|GetTankVehicle)) "Undefined" "MultiplyWithRoot" 0 0.2 0.75 2.5e+03 1.3 0.0 0)))))))

(event EventOnStopWeaponFireFX
  ;[0.4433 0 1] EVENT -> Play/Stop Weapon Fire FX
  (CallFunction|ActivateorDeactivateWindParticle false)
  (if (Effects|ShouldPlayPersistentFireFX)
    (CallFunction|DeactivateMuzzleFX)
    (else
      _))
  (Effects|DeactivateBeamFX)
  (CallFunction|DeactivateAllShellsFX)
  (FortWeapon|StopForceFeedback))

(event Setup|EventOnHostVehicleSetup
  (bind _weaponmesh (Variables|Visual|GetWeaponMesh))
  ;[1 1 1] EVENT -> Setup Host vehicle
  (|CallParentFunction)
  (bind _asfort_athena_vehicle (Utilities|Casting|CastToFortAthenaVehicle (Utilities|ToObject(Interface) (Vehicle|GetHostVehicle))))
  (Variables|Default|SetTankVehicle _asfort_athena_vehicle)
  (Rendering|SetVisibility _weaponmesh false false)
  (Development|SetHiddeninGame _weaponmesh true false) ; Double verifying mesh it hidden
)

(event FortWeapon|EventOnEquippedAndAttached
  ;[1 1 1] EVENT -> On Weapon Equipped and Attached
  ;;[1 1 1] Check if in athena, if not set up fire rate bind.
  (if (not (FortGlobals|IsinAthena))
    (Networking|SwitchHasAuthority
      (:Authority
        _)
      (:Remote
        (CallFunction|BindFireRateChange))))
  (CallFunction|ShowReticle)
  (CallFunction|ShellEjectionFixOn)
  (Utilities|FlowControl|Sequence
    (:then_0
      (CallFunction|InitializeScopeVariables))
    (:then_1
      (Utilities|FlowControl|Delay 0.1) ; This is a hack fix
      (CallFunction|SetActiveAlterationIdleParticles true))))

(event Custom|InitializeScopeVariables
  (bind _scopemesh1p (Variables|Components|GetScopeMesh1P))
  ;[1 1 1] Init Scope Variables
  (bind _instigator (Utilities|GetValidInstigator)
    (:Is%20Valid
      (bind _asfort_pawn (Utilities|Casting|CastToFortPawn _instigator))
      (Utilities|IsValid (Pawn|GetLocalViewingPlayerController _asfort_pawn)
        (:Is%20Valid
          (Networking|CanExecuteCosmeticEvents
            (:True
              (if (Scope|UseScopeTargeting)
                (Utilities|IsValid _scopemesh1p
                  (:Is%20Valid
                    (Rendering|Material|SetVectorParameterValue "/Game/Packages/Fortress_SharedMaterials/GlobalMaterialParameters/FortniteMaterialParameters.FortniteMaterialParameters" "TopLeftScope" (Math|Conversions|ToLinearColor(Vector) (Transformation|GetSocketLocation _scopemesh1p "TopLeftScope")) :!class /Script/Engine.KismetMaterialLibrary)
                    (Rendering|Material|SetVectorParameterValue "/Game/Packages/Fortress_SharedMaterials/GlobalMaterialParameters/FortniteMaterialParameters.FortniteMaterialParameters" "BottomRightScope" (Math|Conversions|ToLinearColor(Vector) (Transformation|GetSocketLocation _scopemesh1p "BottomRightScope")) :!class /Script/Engine.KismetMaterialLibrary)))))))))))

(event Custom|UnhideThirdPersonStuff
  (bind _tankvehicle (Variables|Default|GetTankVehicle))
  (bind _tankvehicle_2 (Variables|Default|GetTankVehicle))
  ;[1 1 1] EVENT -> Set Targeting / Local UnEquip:    Scoped weapon targeting downsights
  ;;[0 0 0] Hiding/showing 3p meshes, etc
  (if (not (Weapon|IsUsingFirstPersonCamera))
    (bind _instigator (Utilities|GetValidInstigator)
      (:Is%20Valid
        (bind _asfort_pawn (Utilities|Casting|CastToFortPawn _instigator))
        (FortWeapon|SetHiddeninGameControllerOnly (Variables|Visual|GetWeaponMesh))
        (PlayerController|SetActorHiddenControllerOnly (Fortnite|GetFortPlayerControllerfromActor _asfort_pawn) _tankvehicle false)
        (Variables|/Script/Engine.SkinnedMeshComponent|SetVisibilityBasedAnimTickOption (Variables|Default|GetDefaultVisibilityBasedAnimTickOption) (Variables|/Script/FortniteGame.FortAthenaVehicle|GetSkeletalMesh _tankvehicle)))))
  (CallFunction|ShowReticle)
  (Utilities|Time|ClearandInvalidateTimerbyHandle (Variables|Default|GetScopeEffectDelay1Handle))
  (Utilities|Time|ClearandInvalidateTimerbyHandle (Variables|Default|GetScopeEffectDelay2Handle))
  (bind _asfort_player_controller_zone (Utilities|Casting|CastToFortPlayerControllerZone (Game|GetInstigatorController))
    (:then
      (Class|FortPlayerController|FixupInteractionWidgetsonUnzoom _asfort_player_controller_zone))
    (:CastFailed
      _))
  (Vehicle|UnbindEventfromOnPawnEnteredVehicleatSeat _tankvehicle_2 (EventDispatchers|CreateEvent :!event Custom|PawnEnterVehicle))
  (Vehicle|UnbindEventfromOnPawnExitedVehiclefromSeat _tankvehicle_2 (EventDispatchers|CreateEvent :!event Custom|PawnExitVehicle))
  (CallFunction|SetAllPassengersVisibility false))

(event EventOnSetTargeting
  (params (bNewIsTargeting :type Boolean))
  ;[1 1 1] EVENT -> Set Targeting / Local UnEquip:    Scoped weapon targeting downsights
  (Networking|CanExecuteCosmeticEvents
    (:True
      (bind _instigator (Utilities|GetValidInstigator)
        (:Is%20Valid
          (bind _asfort_pawn (Utilities|Casting|CastToFortPawn _instigator))
          (Utilities|IsValid (Pawn|GetLocalViewingPlayerController _asfort_pawn)
            (:Is%20Valid
              (bind _asfort_player_controller (Utilities|Casting|CastToFortPlayerController (Pawn|GetLocalViewingPlayerController _asfort_pawn)))
              (Rendering|SetIgnoreSignifanceBasedCustomDepthRendering _asfort_player_controller (Camera|GetIsTargeting) 0)
              (Default|CallOnAimDownSightsChanged bNewIsTargeting)
              (if (Scope|UseScopeTargeting)
                (if bNewIsTargeting
                  (CallFunction|InitializeScopeVariables)
                  (Utilities|FlowControl|Sequence
                    (:then_0
                      (bind _returnvalue (Utilities|Time|SetTimerbyEvent (delegate Custom|ScopeEffectDelay1) 0.1))
                      (Variables|Default|SetScopeEffectDelay1Handle _returnvalue))
                    (:then_1
                      (bind _returnvalue_2 (Utilities|Time|SetTimerbyEvent (delegate Custom|ScopeEffectDelay2) 0.1))
                      (Variables|Default|SetScopeEffectDelay2Handle _returnvalue_2)))
                  (else
                    (Audio|PlayScopeOutAudio)
                    (Utilities|FlowControl|Sequence
                      (:then_0
                        (Audio|StopScopedAudio)
                        (CallFunction|UnhideThirdPersonStuff))
                      (:then_1
                        (CallFunction|ForceScopeBackImmediately))
                      (:then_2
                        (CallFunction|ResetDoonceScopeSound)
                        (CallFunction|SetActiveAlterationIdleParticles true)))))
                (else
                  (CallFunction|ShowReticle))))))))))

(event Custom|HideFirstPersonStuff
  (bind _scopemesh1p (Variables|Components|GetScopeMesh1P))
  ;[1 1 1] EVENT -> Set Targeting / Local UnEquip:    Scoped weapon targeting downsights
  ;;[0 0 0] Hiding/showing 1p meshes, etc
  (Rendering|SetVisibility _scopemesh1p false false)
  (Development|SetHiddeninGame _scopemesh1p true false)
  (Variables|Scope|SetScopePostProcessEnabled false)
  ;;;[0.2865 0.2865 0.2865] Custom Depth
  (if (Variables|Scope|GetScope%20-%20Render%20Enemies%20To%20Custom%20Depth%20Buffer)
    (CallFunction|UpdateEnemyCustomDepths false 0)))

(event Custom|ScopeEffectDelay1
  ;[1 1 1] EVENT -> Set Targeting / Local UnEquip:    Scoped weapon targeting downsights
  (if (Camera|GetIsTargeting)
    (Utilities|FlowControl|Sequence
      (:then_0
        (CallFunction|UnhideFirstPersonStuffPart1)
        (CallFunction|PlayScopePP)
        (CallFunction|SetActiveAlterationIdleParticles false))
      (:then_1
        (Utilities|FlowControl|DoOnce false
          (:Completed
            (Audio|StartScopedAudio)))))))

(event Custom|ResetDoonceScopeSound
  ;[1 1 1] EVENT -> Set Targeting / Local UnEquip:    Scoped weapon targeting downsights
  (Utilities|FlowControl|DoOnce false
    (:Completed
      (Audio|StartScopedAudio))))

(event Custom|HideThirdPersonStuff
  (bind _tankvehicle (Variables|Default|GetTankVehicle))
  (bind _skeletalmesh (Variables|/Script/FortniteGame.FortAthenaVehicle|GetSkeletalMesh _tankvehicle))
  (bind _tankvehicle_2 (Variables|Default|GetTankVehicle))
  ;[1 1 1] EVENT -> Set Targeting / Local UnEquip:    Scoped weapon targeting downsights
  ;;[0 0 0] Hiding/showing 3p meshes, etc
  (bind _instigator (Utilities|GetValidInstigator)
    (:Is%20Valid
      (bind _asfort_pawn (Utilities|Casting|CastToFortPawn _instigator))
      (FortWeapon|SetWeaponVisibilityControllerOnly true true)
      (FortWeapon|SetHiddeninGameControllerOnly (Variables|Visual|GetWeaponMesh) true)
      (PlayerController|SetActorHiddenControllerOnly (Fortnite|GetFortPlayerControllerfromActor _asfort_pawn) _tankvehicle true)
      (Variables|Default|SetDefaultVisibilityBasedAnimTickOption (Variables|/Script/Engine.SkinnedMeshComponent|GetVisibilityBasedAnimTickOption _skeletalmesh))
      (Variables|/Script/Engine.SkinnedMeshComponent|SetVisibilityBasedAnimTickOption "AlwaysTickPoseAndRefreshBones" _skeletalmesh)
      ;;[0 0 0] Hide Passengers and Bind Events to vehicle enter and exit
      (Vehicle|AssignOnPawnEnteredVehicleatSeat _tankvehicle_2 (EventDispatchers|CreateEvent :!event Custom|PawnEnterVehicle))
      (Vehicle|BindEventtoOnPawnExitedVehiclefromSeat _tankvehicle_2 (EventDispatchers|CreateEvent :!event Custom|PawnExitVehicle))
      (CallFunction|SetAllPassengersVisibility true))))

(event Custom|UnhideFirstPersonStuffPart2
  (params (Which%20Call :type Integer))
  (bind _scopemesh1p (Variables|Components|GetScopeMesh1P))
  ;[1 1 1] EVENT -> Set Targeting / Local UnEquip:    Scoped weapon targeting downsights
  ;;[0 0 0] Hiding/showing 1p meshes, etc
  (Rendering|SetVisibility _scopemesh1p true false)
  (Development|SetHiddeninGame _scopemesh1p false false)
  (FortWeapon|SetHiddeninGameOtherLocalControllersOnly _scopemesh1p true)
  ;;;[0.3906 0.3906 0.3906] Custom Depth
  (if (Variables|Scope|GetScope%20-%20Render%20Enemies%20To%20Custom%20Depth%20Buffer)
    (CallFunction|UpdateEnemyCustomDepths true (Variables|Default|GetStencilBufferValue))))

(event Custom|ScopeEffectDelay2
  ;[1 1 1] EVENT -> Set Targeting / Local UnEquip:    Scoped weapon targeting downsights
  (if (Camera|GetIsTargeting)
    (CallFunction|HideThirdPersonStuff)
    (CallFunction|UnhideFirstPersonStuffPart2 2)))

(event FortWeapon|EventOnLocalUnEquip
  ;[1 1 1] EVENT -> Set Targeting / Local UnEquip:    Scoped weapon targeting downsights
  (CallFunction|UnhideThirdPersonStuff)
  (CallFunction|SetActiveAlterationIdleParticles false)
  (CallFunction|DeactivateAllShellsFX)
  (bind _instigator (Utilities|GetValidInstigator)
    (:Is%20Valid
      (bind _asfort_pawn (Utilities|Casting|CastToFortPawn _instigator))
      (Utilities|IsValid (Pawn|GetLocalViewingPlayerController _asfort_pawn)
        (:Is%20Valid
          (if (Scope|UseScopeTargeting)
            (Utilities|FlowControl|Sequence
              (:then_0
                (Audio|StopScopedAudio)
                (CallFunction|UnhideThirdPersonStuff))
              (:then_1
                (CallFunction|ForceScopeBackImmediately))
              (:then_2
                (CallFunction|ResetDoonceScopeSound)
                (CallFunction|SetActiveAlterationIdleParticles true)))
            (else
              (CallFunction|ShowReticle))))))))

(event Custom|UnhideFirstPersonStuffPart1
  ;[1 1 1] EVENT -> Set Targeting / Local UnEquip:    Scoped weapon targeting downsights
  ;;[0 0 0] Hiding/showing 1p meshes, etc
  (Variables|Scope|SetScopePostProcessEnabled true))

(event Custom|PawnEnterVehicle
  (params (Vehicle :type Fort%20Vehicle%20Interface%20Interface%20%28by%20ref%29) (PlayerPawn :type Fort%20Pawn%20Object%20Reference) (SeatIndex :type Integer))
  (bind _hiddenpawns (Variables|Default|GetHiddenPawns))
  ;[1 1 1] EVENT -> Set Targeting / Local UnEquip:    Scoped weapon targeting downsights
  ;;[1 1 1] Set Passenger Visibility on Entering or Exiting the Tank
  (bind _instigator (Utilities|GetValidInstigator)
    (:Is%20Valid
      (if (Pawn|IsLocallyViewed _instigator)
        (FortPawn|SetPawnVisibilityLocal PlayerPawn true true)
        (if (not (Utilities|Array|ContainsItem _hiddenpawns PlayerPawn))
          (Utilities|Array|AddUnique _hiddenpawns PlayerPawn))))))

(event Custom|SetAllPassengersVisibility
  (params (Hide :type Boolean))
  (bind _hiddenpawns (Variables|Default|GetHiddenPawns))
  ;[1 1 1] EVENT -> Set Targeting / Local UnEquip:    Scoped weapon targeting downsights
  ;;[1 1 1] Toggle Driver and Passenger Visibility
  (bind _instigator (Utilities|GetValidInstigator)
    (:Is%20Valid
      (if (Pawn|IsLocallyViewed _instigator)
        (if Hide
          (for _array_element (Vehicle|GetAllPassengers (Variables|Default|GetTankVehicle))
            (if (!= _array_element (Utilities|GetDriverPawn))
              (FortPawn|SetPawnVisibilityLocal _array_element true true)
              (else
                (FortPawn|SetPawnVisibilityLocalControllerOnly _array_element true false)))
            (Utilities|Array|AddUnique (Variables|Default|GetHiddenPawns) _array_element) ; Need to use an array here because the "All Passengers" check fails when the Tank is destroyed.
          )
          (else
            (for _array_element_2 _hiddenpawns
              (if (!= _array_element_2 (Utilities|GetDriverPawn))
                (FortPawn|SetPawnVisibilityLocal _array_element_2 false true)
                (else
                  (FortPawn|SetPawnVisibilityLocalControllerOnly _array_element_2 false false))))
            (Utilities|Array|Clear _hiddenpawns)))))))

(event Custom|ForceScopeBackImmediately
  ;[1 1 1] EVENT -> Set Targeting / Local UnEquip:    Scoped weapon targeting downsights
  ;;[1 1 1] Force Scope Back Immediately
  (bind _instigator (Utilities|GetValidInstigator)
    (:Is%20Valid
      (bind _asfort_player_pawn (Utilities|Casting|CastToFortPlayerPawn _instigator))
      (if (not (Vehicle|IsinVehicle _asfort_player_pawn))
        (CallFunction|SetPostProcessParams 0.0))
      (if (not (Camera|GetIsTargeting))
        (CallFunction|HideFirstPersonStuff)))))

(event Custom|PawnExitVehicle
  (params (Vehicle :type Fort%20Vehicle%20Interface%20Interface%20%28by%20ref%29) (PlayerPawn :type Fort%20Pawn%20Object%20Reference) (SeatIndex :type Integer))
  (bind _hiddenpawns (Variables|Default|GetHiddenPawns))
  ;[1 1 1] EVENT -> Set Targeting / Local UnEquip:    Scoped weapon targeting downsights
  ;;[1 1 1] Set Passenger Visibility on Entering or Exiting the Tank
  (bind _instigator (Utilities|GetValidInstigator)
    (:Is%20Valid
      (if (Pawn|IsLocallyViewed _instigator)
        (if (!= PlayerPawn (Utilities|GetDriverPawn))
          (FortPawn|SetPawnVisibilityLocal PlayerPawn false true)
          (else
            (FortPawn|SetPawnVisibilityLocalControllerOnly PlayerPawn false false)))
        (if (Utilities|Array|ContainsItem _hiddenpawns PlayerPawn)
          (Utilities|Array|RemoveItem _hiddenpawns PlayerPawn))))))

(event EventOnWeaponVisibilityChanged
  (params (bVisible :type Boolean) (bSetForLocalControllerOnly :type Boolean))
  (bind _alteration_ambient_ps (Variables|Effects%28Alterations%29|GetAlteration%20Ambient%20PS))
  ;[1 1 1] EVENT -> OnWeaponVisibilityChanged
  (Utilities|IsValid (Variables|Effects%28Alterations%29|GetAlteration%20Ambient%20PS)
    (:Is%20Valid
      (if bVisible
        (Components|Activation|Activate _alteration_ambient_ps false)
        (else
          (Components|Activation|Deactivate _alteration_ambient_ps)
          (CallFunction|DeactivateAllShellsFX))))))

(event Custom|Update%20Enemy%20Custom%20Depths
  (params (Enable :type Boolean :default true) (StencilBufferValue :type Integer :default 3))
  ;[1 1 1] Update Enemy Custom Depth
  (bind _instigator (Utilities|GetValidInstigator)
    (:Is%20Valid
      (if (Pawn|IsLocallyViewed _instigator)
        _))))

(event Custom|PlayScopePP
  ;[0.1077 1 0] Set Thermal Post Process
  (Utilities|FlowControl|Delay 0.1)
  (if (Camera|GetIsTargeting)
    (CallFunction|SetPostProcessParams 1.0)
    (Components|Timeline|SetPlayRate (Variables|B_Tank_Cannon_Weapon|GetScalePlayerThermal) (/ 1.0 (Variables|Scope|GetPostProcessBlendInTime)))
    (Rendering|Material|SetScalarParameterValue "/Game/Athena/Items/Weapons/Abilities/Thermal/MPC_ThermalGlobal.MPC_ThermalGlobal" "ThermalFadeout" 1.0 :!class /Script/Engine.KismetMaterialLibrary)
    (Utilities|FlowControl|DoOnce false
      (:Completed
        (bind (_direction _newtrack_0) (Timeline|ScalePlayerThermal PlayFromStart)
          (:Update
            (Rendering|Material|SetScalarParameterValue "/Game/Athena/Items/Weapons/Abilities/Thermal/MPC_ThermalGlobal.MPC_ThermalGlobal" "ThermalScale" _newtrack_0 :!class /Script/Engine.KismetMaterialLibrary))
          (:Finished))))))

(event Custom|SetWeaponPierceThrough
  (params (Enable :type Boolean) (TargetLimit :type Integer))
  (bind _returnvalue (Math|Integer|SelectInt TargetLimit 0 Enable))
  ;[1 1 1] Event to dynamically enable/disable trace through pawns
  (FortWeapon|SetTraceThroughPawnsLimit _returnvalue)
  (FortWeapon|SetTraceThroughPawns Enable)
  (CallFunction|SetWeaponPierceThroughClientRep Enable _returnvalue))

(event Custom|SetWeaponPierceThrough_ClientRep
  (params (Enable :type Boolean) (TargetLimit :type Integer))
  ;[1 1 1] Event to dynamically enable/disable trace through pawns
  (FortWeapon|SetTraceThroughPawnsLimit (Math|Integer|SelectInt TargetLimit 0 Enable))
  (FortWeapon|SetTraceThroughPawns Enable))

(event FortWeapon|EventOnPlayImpactFX
  (params (HitResult :type Hit%20Result%20Structure%20%28by%20ref%29) (ImpactPhysicalSurface :type EPhysicalSurface%20Enum :default "SurfaceType_Default") (SpawnedPSC :type FXSystem%20Component%20Object%20Reference))
  (bind (_bblockinghit _binitialoverlap _time _distance _location _impactpoint _normal _impactnormal _physmat _hitactor) (Collision|BreakHitResult HitResult))
  ;[1 1 1] EVENT -> On Play Impact FX:   FORT-182390 Play enemy-only cone attenuated sound from opposite the normal of the surface we're hitting
  (Effects|Components|ParticleSystem|SetIntParameter SpawnedPSC "PhysMatIndex" (Math|Conversions|ToInteger(Byte) ImpactPhysicalSurface))
  (|CallParentFunction HitResult ImpactPhysicalSurface SpawnedPSC)
  (Utilities|Casting|CastToBuildingActor _hitactor)
  (switch Utilities|FlowControl|Switch|SwitchonEFortTeamAffiliation (Team|GetTeamAffiliationtoLocalPlayer (Team|GetActorTeam (Game|GetInstigator)))
    (:Hostile
      (Audio|PlaySoundatLocation "/Game/Athena/Sounds/Weapons/Weapon_Impact_CQCWarning_Enemy_Cue.Weapon_Impact_CQCWarning_Enemy_Cue" (+ _impactpoint (* (Math|Vector|NegateVector _impactnormal) 1e+02)) (Math|Rotator|FindLookatRotation (Transformation|GetWorldLocation (Variables|Visual|GetWeaponMesh)) _impactpoint) 1.0 1.0 0.0 0 0 self))))

(event EventOnApplyFireModeData
  (params (FireModeData :type Fort%20Weapon%20Fire%20Mode%20Data%20Object%20Reference))
  (bind _muzzleniagarasystem (Variables|/Script/FortniteGame.FortWeaponFireModeData|GetMuzzleNiagaraSystem FireModeData))
  (bind _muzzleparticlesystem (Variables|/Script/FortniteGame.FortWeaponFireModeData|GetMuzzleParticleSystem FireModeData))
  ;[1 1 1] EVENT -> Apply Fire Mode Data:   Override muzzle FX from alternative fire mode data.
  (Networking|CanExecuteCosmeticEvents
    (:True
      (CallFunction|DeactivateMuzzleFX) ; stop playing muzzle flash as \nsetting to new particle template \nresets fx.
      (Utilities|FlowControl|Sequence
        (:then_0
          (if (Utilities|IsValid _muzzleparticlesystem :!class /Script/Engine.KismetSystemLibrary)
            (for _array_element (Variables|Effects|GetMuzzleParticleSystemComponents)
              (Effects|Components|ParticleSystem|SetTemplate _array_element _muzzleparticlesystem))))
        (:then_1
          (if (Utilities|IsValid _muzzleniagarasystem :!class /Script/Engine.KismetSystemLibrary)
            (Niagara|SetNiagaraSystemAsset (Variables|Default|GetMuzzleNiagaraComponentInstance) _muzzleniagarasystem true)
            (Variables|Effects|SetMuzzleNiagaraSystemInstance _muzzleniagarasystem)))))))

(event Targeting|EventAbortScopeFX
  ;[1 1 1] EVENT -> Abort Scope FX, ForceScopeFX:    Direct control of scoping, useful when switching between spectated characters
  (Utilities|FlowControl|Sequence
    (:then_0
      (CallFunction|SetPostProcessParams 0.0))
    (:then_1
      (CallFunction|HideFirstPersonStuff))
    (:then_2
      (CallFunction|UnhideThirdPersonStuff))))

(event Targeting|EventForceScopeFX
  ;[1 1 1] EVENT -> Abort Scope FX, ForceScopeFX:    Direct control of scoping, useful when switching between spectated characters
  (Utilities|FlowControl|Sequence
    (:then_0
      (CallFunction|SetPostProcessParams 1.0))
    (:then_1
      (CallFunction|UnhideFirstPersonStuffPart1)
      (CallFunction|UnhideFirstPersonStuffPart2 1))
    (:then_2
      (CallFunction|HideThirdPersonStuff))))

(fn ActivateOrDeactivateWindParticle
  (params (bNewActive :type Boolean))
  (if (Variables|Default|GetIs%20Wind%20Enabled)
    (Components|Activation|SetActive (Variables|Effects|GetMuzzleWindParticleSpawned) bNewActive true))
  (return))

(fn ActivateReloadSmokeFX
  (bind _reload_empty (Variables|Default|GetReload(Empty)))
  (if (Effects|ShouldPlayReloadFX)
    (if (Variables|Default|GetIsReloadNiagara)
      (Components|Activation|Activate (Variables|Default|GetReload_Niagara(Empty)) true)
      (else
        (Utilities|IsValid _reload_empty
          (:Is%20Valid
            (Components|Activation|Activate _reload_empty false)))))))

(fn ActivateShellsFX
  (params (ShouldActivate :type Boolean))
  (if ShouldActivate
    (Networking|CanExecuteCosmeticEvents
      (:True
        (CallFunction|ActivateShellsFXNative "/Game/Weapons/FORT_Generic_Effects/ShellEject/NS_ShellsFire_Burst.NS_ShellsFire_Burst" "/Game/Weapons/FORT_Generic_Effects/ShellEject/NS_ShellsFire_Looping.NS_ShellsFire_Looping" (Variables|Effects%28Shells%29|GetShells%20Socket%20Name) "ShellsRotationRate" (Variables|Effects%28Shells%29|GetShellsRotationRate) "Velocity" (Variables|Effects%28Shells%29|GetShells%20Velocity) "Gravity" (Variables|Effects%28Shells%29|GetShells%20Gravity) "Size" (Variables|Effects%28Shells%29|GetShells%20Size) "InheritParentVelocity" (Math|Conversions|ToVector(Float) (Variables|Effects%28Shells%29|GetInherit%20Parent%20Velocity)) "CylinderRadius" (Variables|Effects%28Shells%29|GetCylindrical%20Radius) "CylinderHeight" (Variables|Effects%28Shells%29|GetCylindrical%20Height) "ShellColor" (Variables|Effects%28Shells%29|GetShell%20Color) "Shell Type" (Utilities|Select :Index (Variables|Effects%28Shells%29|GetShellTypeSelect) :NewEnumerator0 0 :NewEnumerator1 1 :NewEnumerator2 2 :NewEnumerator3 4 :NewEnumerator4 3) (Variables|EffectsDrawDistances|GetShells%20PS%20Max%20Draw%20Distance) (Variables|Effects%28Shells%29|GetUseShellsOnReload?) "SpawnProbability" (Variables|Effects%28Shells%29|GetShells%20Spawn%20Rate%20Scale))))))

(fn AddRandomScale)

(fn DeactivateAllShellsFX
  (Networking|CanExecuteCosmeticEvents
    (:True
      (CallFunction|DeactivateShellsFXNative "SpawnProbability"))))

(fn DeactivateMuzzleFX
  (if (ScalableFloat|GetValueasBool (Variables|Effects|GetUseNativeFX))
    (CallFunction|DeactivateMuzzleFlashFXNative (Variables|Effects|GetMuzzleParticleSystemComponents))
    (else
      (Components|Activation|Deactivate (CallFunction|GetActiveMuzzleComponents)))))

(fn DeactivateReloadSmokeFX
  (bind _reload_empty (Variables|Default|GetReload(Empty)))
  (Utilities|IsValid _reload_empty
    (:Is%20Valid
      (Components|Activation|Deactivate _reload_empty))))

(fn GetActiveMuzzleComponents
  (bind _outniagarasystem (CallFunction|GetCorrectMuzzleNiagaraSystem))
  (bind _output (Utilities|ResolveSoftReference _outniagarasystem))
  (if (Variables|Effects|GetIsMuzzleNiagara)
    (if (Utilities|IsValid (Variables|Default|GetMuzzleNiagaraComponentInstance) :!class /Script/Engine.KismetSystemLibrary)
      ;[1 1 1] Return existing Niagara component
      (return (Utilities|Array|MakeArray (Variables|Default|GetMuzzleNiagaraComponentInstance)))
      (elif (Utilities|IsValid (Variables|Effects|GetMuzzleNiagaraSystemInstance) :!class /Script/Engine.KismetSystemLibrary)
        ;[1 1 1] Create Niagara Component if Needed
        (bind _returnvalue (Niagara|SpawnSystemAttached (Utilities|ResolveSoftReference (CallFunction|GetCorrectMuzzleNiagaraSystem)) (Variables|Visual|GetWeaponMesh) (Effects|GetMuzzleSocketName 0)))
        (bind _output_get (Variables|Default|SetMuzzleNiagaraComponentInstance _returnvalue))
        (return (Utilities|Array|MakeArray _output_get))
        (else
          ;[1 1 1] Load Niagara System If Needed
          (Utilities|IsValid _output
            (:Is%20Valid
              (Variables|Effects|SetMuzzleNiagaraSystemInstance _output))
            (:Is%20Not%20Valid
              (CallFunction|LoadNiagaraMuzzleSoftObject)
              ;;[1 1 1] Just use cascade
              (return (Variables|Effects|GetMuzzleParticleSystemComponents)))))))
    (else
      ;[1 1 1] Just use cascade
      (return (Variables|Effects|GetMuzzleParticleSystemComponents)))))

(fn GetActorAudioParameters
  (bind _returnvalue (Game|GetInstigator))
  (bind _soundparam_islocal (Variables|ActorSoundParam|GetSoundParam_IsLocal))
  (bind _soundparams (Variables|ActorSoundParam|GetSoundParams))
  (bind _output (Utilities|Array|Get(aref) _soundparams _soundparam_islocal))
  (Utilities|FlowControl|Sequence
    (:then_0
      (Utilities|IsValid _returnvalue
        (:Is%20Valid
          (Utilities|Struct|SetmembersinAudioParameter _output (Pawn|IsLocallyViewed _returnvalue)))
        (:Is%20Not%20Valid
          (Utilities|Struct|SetmembersinAudioParameter _output false))))
    (:then_1
      (return (Variables|ActorSoundParam|GetSoundParams)))))

(fn GetCorrectMuzzleNiagaraSystem
  (Utilities|IsValid (Variables|Sounds|GetQuartzGunfireComp)
    (:Is%20Valid
      (return (select (Sounds|IsQuartzGunfireEnabled) (Variables|/Script/FortniteGame.FortQuartzGunfireComponent|GetMuzzleNiagaraSystemLooping (Variables|Sounds|GetQuartzGunfireComp)) (Variables|Effects|GetMuzzleNiagaraSystemAsset))))
    (:Is%20Not%20Valid
      (return (Variables|Effects|GetMuzzleNiagaraSystemAsset)))))

(fn GetScopeParameters
  (return (Variables|Components|GetScopeMesh1P) (Math|Vector2D|MakeVector2D 2e+02 (Variables|Scope|GetTarget%20Scope%20Vignette%20Blur%20Screen%20Percentage)) (Math|Vector|MakeVector (Variables|Scope|GetScopeCameraOffsetX) (Variables|Scope|GetScopeCameraOffsetY) (Variables|Scope|GetScope%20Camera%20Offset%20Amount))))

(fn HideReticle
  (bind _asfort_player_controller_zone (Utilities|Casting|CastToFortPlayerControllerZone (Game|GetInstigatorController)))
  (HUD|SetHUDElementVisibilitywithContext _asfort_player_controller_zone (Variables|Scope|GetReticleHUDElementTags) true "None" ""))

(fn MuzzleFlashFX
  (params (Persistent%20Fire :type Boolean))
  (if (ScalableFloat|GetValueasBool (Variables|Effects|GetUseNativeFX))
    (CallFunction|ActivateMuzzleFlashFXNative Persistent%20Fire (Variables|Effects|GetIsMuzzleNiagara) (CallFunction|GetCorrectMuzzleNiagaraSystem) (Variables|Effects|GetMuzzleParticleSystemComponents) (Variables|Effects|GetMuzzle_ChanceOfLargeFlash) (Variables|Effects|GetMuzzle_FlashLarge_MinScale) (Variables|Effects|GetMuzzle_FlashLarge_MaxScale) (Variables|Effects|GetMuzzle_FlashSmall_MinScale) (Variables|Effects|GetMuzzle_FlashSmall_MaxScale))
    (elif (not (Weapon|IsUsingScope))
      (if (not Persistent%20Fire)
        (CallFunction|AddRandomScale))
      (Components|Activation|Activate (CallFunction|GetActiveMuzzleComponents) true)
      (Variables|Effects|SetLastPlayFXTime (Utilities|Time|GetGameTimeinSeconds))
      (return)
      (else
        _))))

(fn MuzzlePlayReloadFX
  (params (Selection :type EFortReloadFXState%20Enum :default "ReloadStart"))
  (switch Utilities|FlowControl|Switch|SwitchonEFortReloadFXState Selection
    (:ReloadStart
      ;[1 1 1] Start Reload Smoke
      (if (Variables|Effects|GetUse%20Reload%20Particles)
        (CallFunction|ActivateReloadSmokeFX)))
    (:ReloadEnd
      ;[1 1 1] Stop reload smoke
      (if (Variables|Effects|GetUse%20Reload%20Particles)
        (CallFunction|DeactivateReloadSmokeFX)))
    (:Max_None
      ;[1 1 1] Reload Shells
      (CallFunction|ActivateShellsFX (Variables|Effects%28Shells%29|GetUseShellsOnReload?)))))

(fn PlayScopeOutAudio
  (bind _sound_scopezoomout (Variables|Sounds|GetSound_ScopeZoomOut)
    (:then
      (bind _returnvalue (Audio|SpawnSoundAttached _sound_scopezoomout (Variables|Visual|GetWeaponMesh)))
      (Variables|Default|SetScopeZoomOutComp _returnvalue))
    (:else)))

(fn SetActiveAlterationIdleParticles
  (params (Active :type Boolean))
  (Networking|CanExecuteCosmeticEvents
    (:True
      (bind _alteration_ambient_ps (Variables|Effects%28Alterations%29|GetAlteration%20Ambient%20PS)
        (:then
          (if Active
            (Development|SetHiddeninGame _alteration_ambient_ps false false)
            (Components|Activation|Activate _alteration_ambient_ps false)
            (else
              (Development|SetHiddeninGame _alteration_ambient_ps true false)
              (Components|Activation|Deactivate _alteration_ambient_ps)))
          (return))
        (:else)))))

(fn SetActiveMuzzleComponent
  (params (NiagaraEnabled :type Boolean))
  (Variables|Effects|SetIsMuzzleNiagara NiagaraEnabled)
  (CallFunction|LoadNiagaraMuzzleSoftObject))

(fn SetPostProcessParams
  (params (InputPin :type Float%20%28double-precision%29))
  (Variables|Scope|SetScopePostProcessBlendWeight InputPin))

(fn SetWpnRarity)

(fn ShowReticle
  (bind _asfort_player_controller_zone (Utilities|Casting|CastToFortPlayerControllerZone (Game|GetInstigatorController)))
  (HUD|ResetHUDElementVisibilitytoSettings _asfort_player_controller_zone (Variables|Scope|GetReticleHUDElementTags)))

(fn StartScopedAudio
  (Utilities|FlowControl|Sequence
    (:then_0
      (bind _sound_scopezoomin (Variables|Sounds|GetSound_ScopeZoomIn)
        (:then
          (bind _returnvalue (Audio|SpawnSoundAttached _sound_scopezoomin (Variables|Visual|GetWeaponMesh)))
          (Variables|Default|SetScopeZoomInComp _returnvalue))
        (:else)))
    (:then_1
      (bind _sound_scopedinloop (Variables|Sounds|GetSound_ScopedInLoop)
        (:then
          (bind _returnvalue_2 (Audio|SpawnSoundAttached _sound_scopedinloop (Variables|Visual|GetWeaponMesh)))
          (Variables|Default|SetScopedInLoopComp _returnvalue_2))
        (:else)))
    (:then_2
      (bind _scopezoomoutcomp (Variables|Default|GetScopeZoomOutComp)
        (:then
          (Audio|Components|Audio|FadeOut _scopezoomoutcomp 0.25 0.0 "Linear" :!class /Script/Engine.AudioComponent))
        (:else)))))

(fn StopScopedAudio
  (Utilities|FlowControl|Sequence
    (:then_0
      (bind _scopedinloopcomp (Variables|Default|GetScopedInLoopComp)
        (:then
          (Audio|Components|Audio|FadeOut _scopedinloopcomp 0.25 0.0 "Linear" :!class /Script/Engine.AudioComponent))
        (:else)))
    (:then_1
      (bind _scopezoomincomp (Variables|Default|GetScopeZoomInComp)
        (:then
          (Audio|Components|Audio|FadeOut _scopezoomincomp 0.0 0.0 "Linear" :!class /Script/Engine.AudioComponent))
        (:else)))))

(fn ConstructionScript
  (bind _muzzleparticlesystemcomponents (Variables|Effects|GetMuzzleParticleSystemComponents))
  (bind _muzzleparticlesystemcomponents_2 (Variables|Effects|GetMuzzleParticleSystemComponents))
  (Networking|CanExecuteCosmeticEvents
    (:True
      (Utilities|FlowControl|Sequence
        (:then_0
          ;[1 1 1] Initialize MuzzleParticleSystemComponents
          (Utilities|FlowControl|Sequence
            (:then_0
              ;;[1 1 1] Add the existing, default component to the list
              (Utilities|Array|Resize (Variables|Effects|GetMuzzleParticleSystemComponents) (Utilities|Array|Length (Effects|GetUniqueMuzzleSocketNames)))
              (if (Utilities|Array|IsValidIndex _muzzleparticlesystemcomponents_2)
                (Utilities|Array|SetArrayElem _muzzleparticlesystemcomponents_2 0 (Variables|Effects|GetMuzzle(Empty)))))
            (:then_1
              ;;[1 1 1] Create additional components, if there are more muzzle sockets specified
              (for (_ _array_index) (Variables|Effects|GetMuzzleParticleSystemComponents)
                (if (>= _array_index 1)
                  (bind _returnvalue (AddComponent|NODE_AddParticleSystemComponent-2))
                  (if (Utilities|Array|IsValidIndex _muzzleparticlesystemcomponents _array_index)
                    (Utilities|Array|SetArrayElem _muzzleparticlesystemcomponents _array_index _returnvalue)))))
            (:then_2
              ;;[1 1 1] Set up all components
              (for (_array_element _array_index_2) (Variables|Effects|GetMuzzleParticleSystemComponents)
                (Components|Activation|SetAutoActivate _array_element false)
                (Variables|/Script/CascadeRuntime.ParticleSystemComponent|SetAutoAttachSocketName (Utilities|Array|Get(aref) (Effects|GetUniqueMuzzleSocketNames) _array_index_2) _array_element)
                (Effects|Components|ParticleSystem|SetUseAutoManageAttachment _array_element true)
                (Effects|Components|ParticleSystem|SetTemplate _array_element (Variables|Effects|GetMuzzleParticleSystem))))))
        (:then_1
          (CallFunction|SetActiveMuzzleComponent (Effects|IsNiagaraEffectEnabled(SoftObj) "WeaponMuzzleFlashes" (CallFunction|GetCorrectMuzzleNiagaraSystem))))))))
